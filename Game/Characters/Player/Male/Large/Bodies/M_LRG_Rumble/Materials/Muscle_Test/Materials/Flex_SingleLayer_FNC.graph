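# Material function for a single muscle-flex mask layer, kept in a muscle test folder for the large male Rumble player body.
# It passes a 2D Offset input to Atlas_FNC to get UVs, then samples the AtlasMask texture there as linear color.
# The red channel is multiplied by the Intensity scalar and output as the layer's weight.
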MaterialFunction /Game/Characters/Player/Male/Large/Bodies/M_LRG_Rumble/Materials/Muscle_Test/Materials/Flex_SingleLayer_FNC
graph {nodes [
  n1 = MaterialFunctionCall!(Scale, Offset=FunctionInput() {inputName=Offset, inputType=FunctionInput_Vector2}) {materialFunction=&/Atlas_FNC}
  n2 = FunctionInput(Preview=TextureObject()) {inputName=AtlasMask, inputType=FunctionInput_Texture2D, sortPriority=3}
  FunctionOutput(None=Multiply(
    A=FunctionInput() {bUsePreviewValueAsDefault=true, inputName=Intensity, inputType=FunctionInput_Scalar, sortPriority=2}
    B=TextureSample(UVs=@n1.Result, Tex=@n2) {samplerType=SAMPLERTYPE_LinearColor}.R
   ))
 ]}
