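# Material function for the BasilStrong player body that blends a base channel toward a secondary flexed channel, previewed with body normal maps, to show per-muscle flexing.
# Sixteen calls to Flex_SingleLayer_FNC sample a 4x4 mask atlas by cell offset, each scaled by a muscle scalar parameter such as chest, biceps or abdomen.
# The summed masks are clamped, multiplied by a global Muscles_Intensity parameter, and used as the lerp alpha between the two inputs.
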MaterialFunction /Game/Characters/Player/Male/Large/Bodies/M_LRG_BasilStrong/Materials/Basil_Flex_Layers_FNC
graph {nodes [
  n1 = TextureSample() {samplerType=SAMPLERTYPE_Normal, texture=/Game/Characters/Player/Male/Large/Bodies/M_LRG_BasilStrong/Textures/T_M_LRG_BasilStrong_Body_N}
  n2 = TextureSample() {
   samplerType=SAMPLERTYPE_Normal
   texture=/Game/Characters/Player/Male/Large/Bodies/M_LRG_BasilStrong/Textures/T_M_MED_Basil_Strong_Flex_Body_N
  }
  n3 = Constant2Vector()
  n4 = ScalarParameter() {group=Muscle_Flex, parameterName=Chest L, sliderMax=1, sortPriority=1}
  n5 = FunctionInput(Preview=TextureObject() {texture=/Game/Characters/Player/Male/Large/Bodies/M_LRG_BasilStrong/Textures/FlexMasks_Atlas}) {inputName=AtlasMask, inputType=FunctionInput_Texture2D, sortPriority=2}
  n6 = Constant2Vector() {r=1}
  n7 = ScalarParameter() {group=Muscle_Flex, parameterName=Chest R, sliderMax=1, sortPriority=2}
  n8 = Constant2Vector() {r=2}
  n9 = ScalarParameter() {group=Muscle_Flex, parameterName=Forearm_L, sliderMax=1, sortPriority=3}
  n10 = Constant2Vector() {r=3}
  n11 = ScalarParameter() {group=Muscle_Flex, parameterName=Forearm_R, sliderMax=1, sortPriority=4}
  n12 = Add(
   A=Add(
    A=MaterialFunctionCall!(Offset=@n3, Intensity=@n4, AtlasMask=@n5) {materialFunction=/Game/Characters/Player/Male/Large/Bodies/M_LRG_Rumble/Materials/Muscle_Test/Materials/Flex_SingleLayer_FNC}.Result
    B=MaterialFunctionCall(Offset=@n6, Intensity=@n7, AtlasMask=@n5) {materialFunction=&Flex_SingleLayer_FNC}.Result
   )
   B=Add(
    A=MaterialFunctionCall(Offset=@n8, Intensity=@n9, AtlasMask=@n5) {materialFunction=&Flex_SingleLayer_FNC}.Result
    B=MaterialFunctionCall(Offset=@n10, Intensity=@n11, AtlasMask=@n5) {materialFunction=&Flex_SingleLayer_FNC}.Result
   )
  )
  n13 = Constant2Vector() {g=1}
  n14 = ScalarParameter() {group=Muscle_Flex, parameterName=Deltoid_L, sliderMax=1, sortPriority=5}
  n15 = Constant2Vector() {g=1, r=1}
  n16 = ScalarParameter() {group=Muscle_Flex, parameterName=Deltoid_R, sliderMax=1, sortPriority=6}
  n17 = Constant2Vector() {g=1, r=2}
  n18 = ScalarParameter() {group=Muscle_Flex, parameterName=Back_L, sliderMax=1, sortPriority=7}
  n19 = Constant2Vector() {g=1, r=3}
  n20 = ScalarParameter() {group=Muscle_Flex, parameterName=Back_R, sliderMax=1, sortPriority=8}
  n21 = Add(
   A=Add(
    A=MaterialFunctionCall(Offset=@n13, Intensity=@n14, AtlasMask=@n5) {materialFunction=&Flex_SingleLayer_FNC}.Result
    B=MaterialFunctionCall(Offset=@n15, Intensity=@n16, AtlasMask=@n5) {materialFunction=&Flex_SingleLayer_FNC}.Result
   )
   B=Add(
    A=MaterialFunctionCall(Offset=@n17, Intensity=@n18, AtlasMask=@n5) {materialFunction=&Flex_SingleLayer_FNC}.Result
    B=MaterialFunctionCall(Offset=@n19, Intensity=@n20, AtlasMask=@n5) {materialFunction=&Flex_SingleLayer_FNC}.Result
   )
  )
  n22 = Constant2Vector() {g=2}
  n23 = ScalarParameter() {group=Muscle_Flex, parameterName=Biceps_L, sliderMax=1, sortPriority=9}
  n24 = Constant2Vector() {g=2, r=1}
  n25 = ScalarParameter() {group=Muscle_Flex, parameterName=Biceps_R, sliderMax=1, sortPriority=10}
  n26 = Constant2Vector() {g=2, r=2}
  n27 = ScalarParameter() {group=Muscle_Flex, parameterName=2_2, sliderMax=1, sortPriority=11}
  n28 = Constant2Vector() {g=2, r=3}
  n29 = ScalarParameter() {group=Muscle_Flex, parameterName=3_2, sliderMax=1, sortPriority=12}
  n30 = Add(
   A=Add(
    A=MaterialFunctionCall(Offset=@n22, Intensity=@n23, AtlasMask=@n5) {materialFunction=&Flex_SingleLayer_FNC}.Result
    B=MaterialFunctionCall(Offset=@n24, Intensity=@n25, AtlasMask=@n5) {materialFunction=&Flex_SingleLayer_FNC}.Result
   )
   B=Add(
    A=MaterialFunctionCall(Offset=@n26, Intensity=@n27, AtlasMask=@n5) {materialFunction=&Flex_SingleLayer_FNC}.Result
    B=MaterialFunctionCall(Offset=@n28, Intensity=@n29, AtlasMask=@n5) {materialFunction=&Flex_SingleLayer_FNC}.Result
   )
  )
  n31 = Constant2Vector() {g=3}
  n32 = ScalarParameter() {group=Muscle_Flex, parameterName=Triceps_L, sliderMax=1, sortPriority=13}
  n33 = Constant2Vector() {g=3, r=1}
  n34 = ScalarParameter() {group=Muscle_Flex, parameterName=Triceps_R, sliderMax=1, sortPriority=14}
  n35 = Constant2Vector() {g=3, r=2}
  n36 = ScalarParameter() {group=Muscle_Flex, parameterName=Abdomen, sliderMax=1, sortPriority=15}
  n37 = Constant2Vector() {g=3, r=3}
  n38 = ScalarParameter() {group=Muscle_Flex, parameterName=3_3, sliderMax=1, sortPriority=16}
  n39 = Add(
   A=Add(
    A=MaterialFunctionCall(Offset=@n31, Intensity=@n32, AtlasMask=@n5) {materialFunction=&Flex_SingleLayer_FNC}.Result
    B=MaterialFunctionCall(Offset=@n33, Intensity=@n34, AtlasMask=@n5) {materialFunction=&Flex_SingleLayer_FNC}.Result
   )
   B=Add(
    A=MaterialFunctionCall(Offset=@n35, Intensity=@n36, AtlasMask=@n5) {materialFunction=&Flex_SingleLayer_FNC}.Result
    B=MaterialFunctionCall(Offset=@n37, Intensity=@n38, AtlasMask=@n5) {materialFunction=&Flex_SingleLayer_FNC}.Result
   )
  )
  n40 = Clamp(None=Add(A=Add(A=@n12, B=@n21), B=Add(A=@n30, B=@n39)))
  n41 = ScalarParameter() {
   defaultValue=1
   desc=Multiplies the intensity of the muscle masks. Useful for reducing global muscle intensity
   group=Muscle_Flex
   parameterName=Muscles_Intensity
   sliderMax=1
   sortPriority=0
  }
  FunctionOutput(None=LinearInterpolate(
    A=FunctionInput(Preview=@n1.RGB) {inputName=BaseChannel}
    B=FunctionInput(Preview=@n2.RGB) {inputName=SecondaryChannel, sortPriority=1}
    Alpha=Multiply(A=@n40, B=@n41)
   ))
 ]}
