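# Parent material for the Basil Strong large male player body, adding a muscle-flex normal layer to the character shading, for skeletal meshes with morph targets and mesh deformers.
# The shared character PreFX function builds the base material attributes, from which the Normal is extracted.
# Basil_Flex_Layers_FNC combines that normal with a flex normal map using a mask atlas parameter.
# The result replaces the Normal attribute and feeds the shared PostFX function.
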Material /Game/Characters/Player/Male/Large/Bodies/M_LRG_BasilStrong/Materials/M_Basil_Strong_Parent
graph {
 properties {
  bUseMaterialAttributes=true
  bUsedWithMeshDeformer=true
  bUsedWithMorphTargets=true
  bUsedWithSkeletalMesh=true
  pixelDepthOffsetMode=PDOM_Legacy
  usedShadingModels=Default Lit
 }
 nodes [
  n1 = MaterialFunctionCall() {materialFunction=&BaseFN_CharacterFunction_PreFX}
  n2 = GetMaterialAttributes(None=@n1.Result) {attributeGetTypes [#1]}
  n3 = TextureSampleParameter2D() {
   group=Flex
   parameterName=Flex Normal Map
   samplerType=SAMPLERTYPE_Normal
   texture=/Game/Characters/Player/Male/Large/Bodies/M_LRG_BasilStrong/Textures/T_M_LRG_BasilStrongFlex_Body_N
  }
  n4 = TextureObjectParameter() {group=Flex, parameterName=FlexMaskAtlas, texture=/Game/Characters/Player/Male/Large/Bodies/M_LRG_BasilStrong/Textures/FlexMasks_Atlas}
  MaterialFunctionCall(
   AO=@n1.AO
   Distance Blend=@n1.AO
   Fresnel=@n1.AO
   Gmap=@n1.AO
   Pre Skinned Local Position=@n1.AO
   Material=SetMaterialAttributes!(
    MaterialAttributes=@n2.MaterialAttributes
    Normal=MaterialFunctionCall!(BaseChannel=@n2.Normal, SecondaryChannel=@n3.RGB, AtlasMask=@n4) {materialFunction=&/Basil_Flex_Layers_FNC}.Result
   ) {attributeSetTypes [#1]}
  ) {materialFunction=&BaseFN_CharacterFunction_PostFX}
  TextureSample() {texture=/Game/Characters/Player/Male/Large/Bodies/M_LRG_BasilStrong/Textures/Flexmax_Basil_Atlas_M}
 ]
}
