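# Physics asset for the left leg mesh of the large male BasilStrong player body, covering the leg and four dynamic knee flaps.
# Capsules on thigh, thigh twist and calf form the leg volume, with calf and twist constrained to the thigh.
# Flaps A–D are three-bone capsule chains whose bases attach to the calf with free swing and twist.
# Segments within each chain have narrow swing limits and zero twist limit.
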PhysicsAsset /Game/Characters/Player/Male/Large/Bodies/M_LRG_BasilStrong/Meshes/M_LRG_BasilStrong_LegLf_Physics
bodies {
 calf_l [{shapeType=Capsule, length=21.8043, radius=8.5, center=<-14.5477 -2.7216 -0.7133>, rotation=<rot 0 95.698 -88.6773>}]
 dyn_knee_l_flapA_01 [{shapeType=Capsule, length=4, radius=2, shapeName=dyn_knee_l_flapA_01_capsule, center=<1.5504 0.5278 0>}]
 dyn_knee_l_flapA_02 [{shapeType=Capsule, length=4, radius=2, shapeName=dyn_knee_l_flapA_02_capsule, center=<2.8522 0.363 0>, rotation=<rot 90 0 0>}]
 dyn_knee_l_flapA_base [{shapeType=Capsule, length=0.505, radius=0.505, shapeName=dyn_knee_l_flapA_base_capsule, rotation=<rot 0 -90 0>}]
 dyn_knee_l_flapB_01 [{shapeType=Capsule, length=4, radius=2, shapeName=dyn_knee_l_flapB_01_capsule, center=<1.5504 0.5278 0>}]
 dyn_knee_l_flapB_02 [{shapeType=Capsule, length=4, radius=2, shapeName=dyn_knee_l_flapB_02_capsule, center=<2.8522 0.363 0>, rotation=<rot 90 0 0>}]
 dyn_knee_l_flapB_base [{shapeType=Capsule, length=0.505, radius=0.505, shapeName=dyn_knee_l_flapB_base_capsule, rotation=<rot 0 -90 0>}]
 dyn_knee_l_flapC_01 [{shapeType=Capsule, length=4, radius=3, shapeName=dyn_knee_l_flapC_01_capsule, center=<2.9278 0.7926 0>, rotation=<rot 90 0 0>}]
 dyn_knee_l_flapC_02 [{
   shapeType=Capsule
   length=0.505
   radius=0.505
   shapeName=dyn_knee_l_flapC_02_capsule
   center=<1.0317 0.0848 0.0901>
   rotation=<rot 0 -84.8864 -73.8669>
  }]
 dyn_knee_l_flapC_base [{shapeType=Capsule, length=0.505, radius=0.505, shapeName=dyn_knee_l_flapC_base_capsule, rotation=<rot 0 -90 0>}]
 dyn_knee_l_flapD_01 [{shapeType=Capsule, length=2, radius=3, shapeName=dyn_knee_l_flapD_01_capsule, center=<3.8398 1.8029 0>, rotation=<rot 90 0 0>}]
 dyn_knee_l_flapD_02 [{
   shapeType=Capsule
   length=2.3868
   radius=3
   shapeName=dyn_knee_l_flapD_02_capsule
   center=<2.9155 1.5228 0.2356>
   rotation=<rot 0 94.4988 -69.5898>
  }]
 dyn_knee_l_flapD_base [{shapeType=Capsule, length=0.505, radius=0.505, shapeName=dyn_knee_l_flapD_base_capsule, rotation=<rot 0 -90 0>}]
 thigh_l [{shapeType=Capsule, length=20.5796, radius=14.3578, center=<-19.9518 0.2757 1.4678>, rotation=<rot 0 -84.1933 -84.6497>}]
 thigh_twist_01_l [{shapeType=Capsule, length=19.9735, radius=14.2879, center=<-0.4927 1.1683 1.1927>, rotation=<rot 0 -90.4215 -71.5205>}]
}
constraints [
 {bone1Name=calf_l, bone2Name=thigh_l, swing1LimitDegrees=45, swing2LimitDegrees=45, twistLimitDegrees=45}
 {bone1Name=thigh_twist_01_l, bone2Name=thigh_l, swing1LimitDegrees=45, swing2LimitDegrees=45, twistLimitDegrees=45}
 {
  bone1Name=dyn_knee_l_flapA_base
  bone2Name=calf_l
  swing1LimitDegrees=45
  swing1Motion=Free
  swing2LimitDegrees=45
  swing2Motion=Free
  twistLimitDegrees=45
  twistMotion=Free
 }
 {bone1Name=dyn_knee_l_flapA_01, bone2Name=dyn_knee_l_flapA_base, swing1LimitDegrees=15, swing2LimitDegrees=10, twistLimitDegrees=0}
 {bone1Name=dyn_knee_l_flapA_02, bone2Name=dyn_knee_l_flapA_01, swing1LimitDegrees=15, swing2LimitDegrees=10, twistLimitDegrees=0}
 {bone1Name=dyn_knee_l_flapB_01, bone2Name=dyn_knee_l_flapB_base, swing1LimitDegrees=15, swing2LimitDegrees=10, twistLimitDegrees=0}
 {
  bone1Name=dyn_knee_l_flapB_base
  bone2Name=calf_l
  swing1LimitDegrees=45
  swing1Motion=Free
  swing2LimitDegrees=45
  swing2Motion=Free
  twistLimitDegrees=45
  twistMotion=Free
 }
 {bone1Name=dyn_knee_l_flapB_02, bone2Name=dyn_knee_l_flapB_01, swing1LimitDegrees=15, swing2LimitDegrees=10, twistLimitDegrees=0}
 {
  bone1Name=dyn_knee_l_flapC_base
  bone2Name=calf_l
  swing1LimitDegrees=45
  swing1Motion=Free
  swing2LimitDegrees=45
  swing2Motion=Free
  twistLimitDegrees=45
  twistMotion=Free
 }
 {bone1Name=dyn_knee_l_flapC_01, bone2Name=dyn_knee_l_flapC_base, swing1LimitDegrees=15, swing2LimitDegrees=10, twistLimitDegrees=0}
 {
  bone1Name=dyn_knee_l_flapD_base
  bone2Name=calf_l
  swing1LimitDegrees=45
  swing1Motion=Free
  swing2LimitDegrees=45
  swing2Motion=Free
  twistLimitDegrees=45
  twistMotion=Free
 }
 {bone1Name=dyn_knee_l_flapD_01, bone2Name=dyn_knee_l_flapD_base, swing1LimitDegrees=15, swing2LimitDegrees=10, twistLimitDegrees=0}
 {bone1Name=dyn_knee_l_flapC_02, bone2Name=dyn_knee_l_flapC_01, swing1LimitDegrees=15, swing2LimitDegrees=10, twistLimitDegrees=0}
 {bone1Name=dyn_knee_l_flapD_02, bone2Name=dyn_knee_l_flapD_01, swing1LimitDegrees=15, swing2LimitDegrees=10, twistLimitDegrees=0}
]
