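# Physics asset for the right leg mesh of the large male BasilStrong player body, covering the leg and three dynamic knee flaps.
# Capsules on the thigh, thigh twist and calf form the leg volumes, with calf and twist constrained to the thigh.
# Flaps A, B and C are three-body capsule chains whose tiny base bodies attach to the calf with free rotation.
# The two segments below each base have narrow swing limits and zero twist range.
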PhysicsAsset /Game/Characters/Player/Male/Large/Bodies/M_LRG_BasilStrong/Meshes/M_LRG_BasilStrong_LegRt_Physics
bodies {
 calf_r [{shapeType=Capsule, length=21.206, radius=8, center=<14.8291 2.0897 -0.2662>, rotation=<rot 0 90.8171 -88.5478>}]
 dyn_knee_r_flapA_01 [{
   shapeType=Capsule
   length=4
   radius=3
   shapeName=dyn_knee_r_flapA_01_capsule
   center=<3.2215 -1.5189 0.0025>
   rotation=<rot 0 179.6202 179.6204>
  }]
 dyn_knee_r_flapA_02 [{shapeType=Capsule, length=2.9656, radius=2, shapeName=dyn_knee_r_flapA_02_capsule, center=<3.5832 -1.093 0>, rotation=<rot -90 -0.0002 0>}]
 dyn_knee_r_flapA_base [{shapeType=Capsule, length=0.505, radius=0.505, shapeName=dyn_knee_r_flapA_base_capsule, rotation=<rot 0 -90 0>}]
 dyn_knee_r_flapB_01 [{shapeType=Capsule, length=7, radius=3, shapeName=dyn_knee_r_flapB_01_capsule, center=<4.372 -2.0386 0>}]
 dyn_knee_r_flapB_02 [{shapeType=Capsule, length=2.9746, radius=2, shapeName=dyn_knee_r_flapB_02_capsule, center=<3.8299 -0.9962 0>, rotation=<rot 90 0 0>}]
 dyn_knee_r_flapB_base [{shapeType=Capsule, length=0.505, radius=0.505, shapeName=dyn_knee_r_flapB_base_capsule, rotation=<rot 0 -90 0>}]
 dyn_knee_r_flapC_01 [{shapeType=Capsule, length=4.697, radius=2, shapeName=dyn_knee_r_flapC_01_capsule, center=<4.2398 -1.0283 0>, rotation=<rot 90 0 0>}]
 dyn_knee_r_flapC_02 [{shapeType=Capsule, length=3.0356, radius=2, shapeName=dyn_knee_r_flapC_02_capsule, center=<4.2398 -0.9589 0>, rotation=<rot 90 0 0>}]
 dyn_knee_r_flapC_base [{shapeType=Capsule, length=0.505, radius=0.505, shapeName=dyn_knee_r_flapC_base_capsule, rotation=<rot 0 -90 0>}]
 thigh_r [{shapeType=Capsule, length=21.0374, radius=14.2905, center=<20.7532 -0.6348 -0.7289>, rotation=<rot 0 -85.7757 -77.0339>}]
 thigh_twist_01_r [{shapeType=Capsule, length=20.5811, radius=13.8862, center=<0.7334 -1.3597 -0.188>, rotation=<rot 0 -93.0837 -68.1231>}]
}
constraints [
 {bone1Name=calf_r, bone2Name=thigh_r, swing1LimitDegrees=45, swing2LimitDegrees=45, twistLimitDegrees=45}
 {bone1Name=thigh_twist_01_r, bone2Name=thigh_r, swing1LimitDegrees=45, swing2LimitDegrees=45, twistLimitDegrees=45}
 {
  bone1Name=dyn_knee_r_flapA_base
  bone2Name=calf_r
  swing1LimitDegrees=45
  swing1Motion=Free
  swing2LimitDegrees=45
  swing2Motion=Free
  twistLimitDegrees=45
  twistMotion=Free
 }
 {bone1Name=dyn_knee_r_flapA_01, bone2Name=dyn_knee_r_flapA_base, swing1LimitDegrees=15, swing2LimitDegrees=20, twistLimitDegrees=0}
 {bone1Name=dyn_knee_r_flapA_02, bone2Name=dyn_knee_r_flapA_01, swing1LimitDegrees=15, swing2LimitDegrees=20, twistLimitDegrees=0}
 {
  bone1Name=dyn_knee_r_flapB_base
  bone2Name=calf_r
  swing1LimitDegrees=45
  swing1Motion=Free
  swing2LimitDegrees=45
  swing2Motion=Free
  twistLimitDegrees=45
  twistMotion=Free
 }
 {bone1Name=dyn_knee_r_flapB_01, bone2Name=dyn_knee_r_flapB_base, swing1LimitDegrees=15, swing2LimitDegrees=20, twistLimitDegrees=0}
 {bone1Name=dyn_knee_r_flapB_02, bone2Name=dyn_knee_r_flapB_01, swing1LimitDegrees=15, swing2LimitDegrees=20, twistLimitDegrees=0}
 {
  bone1Name=dyn_knee_r_flapC_base
  bone2Name=calf_r
  swing1LimitDegrees=45
  swing1Motion=Free
  swing2LimitDegrees=45
  swing2Motion=Free
  twistLimitDegrees=45
  twistMotion=Free
 }
 {bone1Name=dyn_knee_r_flapC_01, bone2Name=dyn_knee_r_flapC_base, swing1LimitDegrees=15, swing2LimitDegrees=20, twistLimitDegrees=0}
 {bone1Name=dyn_knee_r_flapC_02, bone2Name=dyn_knee_r_flapC_01, swing1LimitDegrees=15, swing2LimitDegrees=20, twistLimitDegrees=0}
]
