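# Animation Blueprint for the BasilStrong body that follows the parent mesh's pose and adds secondary motion and material-driven muscle flexing.
# The copied pose runs through a deform Control Rig, per-leg RigidBody physics with gravity override, crouch arm fixes and knee-flap offsets.
# At LOD 1 or lower, bone angles and animation curves feed MuscleFNC, which curve-remaps them into material parameters, exaggerated during the BasilStrong emote.
# The Knees rig and a second muscle chain are unconnected.
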AnimBlueprint /Game/Characters/Player/Male/Large/Bodies/M_LRG_BasilStrong/Meshes/M_LRG_BasilStrong_AnimBP
target_skeleton=&/M_LRG_BasilStrong_Skeleton
graphs [
 AnimGraph {nodes [
   Control Rig() {
    alphaBoolBlend {blendOption=Linear}
    controlRigAssetReference {blueprintRigClass=&/M_LRG_BasilStrong_Knees_CtrlRig.M_LRG_BasilStrong_Knees_CtrlRig_C}
   }
   n1 = Control Rig(Source=Copy Pose From Mesh() {bUseAttachedParent=true}) {
    alphaBoolBlend {blendOption=Linear}
    controlRigAssetReference {blueprintRigClass=&/M_LRG_BasilStrong_DeformRig_Master_ControlRig.M_LRG_BasilStrong_DeformRig_Master_ControlRig_C}
   }
   n2 = Get GravityOverride()
   n3 = RigidBody!(ComponentPose=Local To Component(LocalPose=@n1), Alpha, OverrideWorldGravity:Vector=@n2, ExternalForce) -> (Pose) {
    bClampLinearTranslationLimitToRefPose=true
    bOverrideWorldGravity=true
    overridePhysicsAsset=&/M_LRG_BasilStrong_LegLf_Physics
    simulationSpace=BaseBoneSpace
    alphaBoolBlend {blendOption=Linear}
    baseBoneRef {boneName=thigh_l}
    componentAppliedLinearAccClamp=<5000 5000 5000>
    componentLinearAccScale {z=0.6}
    componentLinearVelScale {z=0.6}
    overrideWorldGravity {z=-980}
    simSpaceSettings {
     maxAngularAcceleration=500
     maxAngularVelocity=3
     maxLinearAcceleration=700
     maxLinearVelocity=500
     velocityScaleZ=0
     worldAlpha=0.25
     externalLinearDragV=<1 1>
    }
   }
   n4 = RigidBody(ComponentPose=@n3, OverrideWorldGravity=@n2) {
    bClampLinearTranslationLimitToRefPose=true
    bOverrideWorldGravity=true
    overridePhysicsAsset=&/M_LRG_BasilStrong_LegRt_Physics
    simulationSpace=BaseBoneSpace
    alphaBoolBlend {blendOption=Linear}
    baseBoneRef {boneName=thigh_r}
    componentAppliedLinearAccClamp=<5000 5000 5000>
    componentLinearAccScale {z=0.6}
    componentLinearVelScale {z=0.6}
    overrideWorldGravity {z=-980}
    simSpaceSettings {
     maxAngularAcceleration=500
     maxAngularVelocity=3
     maxLinearAcceleration=700
     maxLinearVelocity=500
     velocityScaleZ=0
     worldAlpha=0.25
     externalLinearDragV=<1 1>
    }
   }
   n5 = Get bIsCrouching()
   n6 = "Transform (Modify) Bone - Bone: upperarm_twist_01_r"!(ComponentPose=@n4, bAlphaBoolEnabled:bool=@n5, Translation="0,0,-2", Rotation="0,0,0", Scale="1,1,1") -> (Pose) {
    alphaInputType=Bool
    translationMode=BMM_Additive
    translationSpace=BCS_BoneSpace
    alphaBoolBlend {blendOption=Linear}
    boneToModify {boneName=upperarm_twist_01_r}
    translation {z=-2}
   }
   n7 = "Transform (Modify) Bone - Bone: upperarm_twist_02_r"!(ComponentPose=@n6, bAlphaBoolEnabled:bool=@n5, Translation="0,0,-3", Rotation="0,-30,0", Scale="1,1,1") -> (Pose) {
    alphaInputType=Bool
    rotationMode=BMM_Additive
    translationMode=BMM_Additive
    translationSpace=BCS_BoneSpace
    alphaBoolBlend {blendInTime=0.2, blendOption=Linear, blendOutTime=0.2}
    boneToModify {boneName=upperarm_twist_02_r}
    rotation {yaw=-30}
    translation {z=-3}
   }
   n8 = "Transform (Modify) Bone - Bone: dyn_knee_r_flap_master"(ComponentPose=@n7, Alpha=1, Translation="0,-2,1", Rotation="0,0,0", Scale="1,1,1") {
    translationMode=BMM_Additive
    translationSpace=BCS_BoneSpace
    alphaBoolBlend {blendOption=Linear}
    boneToModify {boneName=dyn_knee_r_flap_master}
    translation {y=-2, z=1}
   }
   n9 = "Transform (Modify) Bone - Bone: dyn_knee_l_flap_master"(ComponentPose=@n8, Alpha=1, Translation="0,2,0", Rotation="0,0,0", Scale="1,1,1") {
    translationMode=BMM_Additive
    translationSpace=BCS_BoneSpace
    alphaBoolBlend {blendOption=Linear}
    boneToModify {boneName=dyn_knee_l_flap_master}
    translation {y=2}
   }
   Output Pose(Result=Component To Local(ComponentPose=@n9)) {layerGroup=None}
  ]}
 EventGraph {nodes [
   Event BlueprintUpdateAnimation()
   > n1 = Sequence!(execute) -> (then_0, then_1, then_2:Exec)
   n2 = Get CurrentLOD()
   > Branch(Condition=AND Boolean(A=Get IsMuscleON() -> (IsMuscleON:bool), B="integer <= integer"(A=@n2, B=1)))
   > Sequence!(execute) -> (then_0, then_1, then_2, then_3:Exec, then_4:Exec)
   > n3 = Sequence()
   n4 = GetOwnerSkeletalMeshComponent()
   n5 = Branch(
    execute=@n3.then_3
    Condition=Montage_IsPlaying(self=GetAnimInstance(self=@n4), Montage=/Game/Animation/Game/MainPlayer/Emotes/BasilStrong/Emote_BasilStrong_CLM_M)
   )
   > Set muscleFullON(muscleFullON:bool=true) -> (Output_Get:bool)
   Set muscleFullON(execute=@n5.else, muscleFullON=false)
   n6 = Get Dynamic Material Instance(execute=@n3.then_4, SkeletalMeshComponent=GetPartSkeletalMeshComponent())
   > n7 = Sequence!(execute) -> (then_0, then_1, then_2, then_3, then_4, then_5:Exec, then_6:Exec)
   n8 = Get elbow_rot_l() -> (elbow_rot_l:double)
   n9 = SelectFloat(A=float * float(A=4, B=@n8) {is=title K2Node_PromotableOperator}, B=@n8, bPickA=Get muscleFullON() -> (muscleFullON:bool))
   n10 = Get easyInOutCrv_var() -> (easyInOutCrv_var:Curve Float Object Reference)
   n11 = MuscleFNC!(
    execute=@n7.then_1
    self
    inTarget:Material Instance Dynamic Object Reference=@n6
    paramName:Name=Biceps_L
    value:double=@n9
    inRangeA:double=0
    inRangeB:double=95
    curve:Curve Float Object Reference=@n10
    alpha:double=1
   ) -> (then, outValue:double, outTarget:Material Instance Dynamic Object Reference)
   > n12 = MuscleFNC(inTarget=@n11.outTarget, paramName=Triceps_L, value=@n9, inRangeA=0, inRangeB=-20, curve=@n10, alpha=1)
   n13 = Get elbow_rot_r() -> (elbow_rot_r:double)
   n14 = SelectFloat(A=float * float(A=6, B=@n13), B=@n13, bPickA=Get muscleFullON())
   n15 = Get easyInOutCrv_var()
   > n16 = MuscleFNC(inTarget=@n12.outTarget, paramName=Biceps_R, value=@n14, inRangeA=0, inRangeB=95, curve=@n15, alpha=1)
   > n17 = MuscleFNC(inTarget=@n16.outTarget, paramName=Triceps_R, value=@n14, inRangeA=0, inRangeB=-20, curve=@n15, alpha=1)
   n18 = Get elbow_rot_l()
   n19 = float + float(
    A=GetCurveValue(CurveName=l_arm_up)
    B=float * float(A=Remap Float(Input=@n18, OldRange=(X=0,Y=100), NewRange=(X=0,Y=1), Clamp=true) -> (Output=~), B=0.5)
   )
   n20 = MuscleFNC(
    execute=@n7.then_2
    inTarget=@n6
    paramName=Deltoid_L
    value=SelectFloat(A=float * float(A=3, B=@n19), B=@n19, bPickA=Get muscleFullON())
    inRangeA=0
    inRangeB=1
    curve=Get easyInOutCrv_var()
    alpha=1
   )
   n21 = Get elbow_rot_r()
   n22 = float + float(
    A=GetCurveValue(CurveName=r_arm_up)
    B=float * float(A=Remap Float(Input=@n21, OldRange=(X=0,Y=100), NewRange=(X=0,Y=1), Clamp=true) -> (Output=~), B=0.5)
   )
   > n23 = MuscleFNC(
    inTarget=@n20.outTarget
    paramName=Deltoid_R
    value=SelectFloat(A=float * float(A=4, B=@n22), B=@n22, bPickA=Get muscleFullON())
    inRangeA=0
    inRangeB=1
    curve=Get easyInOutCrv_var()
    alpha=1
   )
   n24 = Get clavicle_upDn_l() -> (clavicle_upDn_l:double)
   n25 = Get clavicle_fwdBk_l() -> (clavicle_fwdBk_l:double)
   > n26 = MuscleFNC(
    inTarget=@n23.outTarget
    paramName=Back_L
    value=float + float!(
     A=GetCurveValue(CurveName=l_arm_bk)
     B=Remap Float(Input=@n25, OldRange=(X=0,Y=-15), NewRange=(X=0,Y=1), Clamp=true) -> (Output=~)
     C:double=Remap Float(Input=@n24, OldRange=(X=0,Y=40), NewRange=(X=0,Y=1), Clamp=true) -> (Output=~)
    ) -> (ReturnValue)
    inRangeA=0
    inRangeB=1.25
    curve=Get easyInOutCrv_var()
    alpha=1
   )
   n27 = Get clavicle_upDn_r() -> (clavicle_upDn_r:double)
   n28 = Get clavicle_fwdBk_r() -> (clavicle_fwdBk_r:double)
   > MuscleFNC(
    inTarget=@n26.outTarget
    paramName=Back_R
    value=float + float(
     A=GetCurveValue(CurveName=r_arm_bk)
     B=Remap Float(Input=@n28, OldRange=(X=0,Y=-15), NewRange=(X=0,Y=1), Clamp=true) -> (Output=~)
     C=Remap Float(Input=@n27, OldRange=(X=0,Y=40), NewRange=(X=0,Y=1), Clamp=true) -> (Output=~)
    )
    inRangeA=0
    inRangeB=1.25
    curve=Get easyInOutCrv_var()
    alpha=1
   )
   n29 = float + float!(
    A=float * float(A=GetCurveValue(CurveName=l_arm_fwd), B=0.5)
    B=Remap Float(Input=Get clavicle_fwdBk_l(), OldRange=(X=0,Y=25), NewRange=(X=0,Y=1), Clamp=true) -> (Output=~)
   ) -> (ReturnValue)
   n30 = MuscleFNC(
    execute=@n7.then_0
    inTarget=@n6
    paramName=Chest L
    value=SelectFloat(A=float * float(A=5, B=@n29), B=@n29, bPickA=Get muscleFullON())
    inRangeA=0
    inRangeB=1.5
    curve=Get easyInOutCrv_var()
    alpha=1
   )
   n31 = float + float(
    A=float * float(A=GetCurveValue(CurveName=r_arm_fwd), B=0.5)
    B=Remap Float(Input=Get clavicle_fwdBk_r(), OldRange=(X=0,Y=25), NewRange=(X=0,Y=1), Clamp=true) -> (Output=~)
   )
   > MuscleFNC(
    inTarget=@n30.outTarget
    paramName=Chest R
    value=SelectFloat(A=float * float(A=3, B=@n31), B=@n31, bPickA=Get muscleFullON())
    inRangeA=0
    inRangeB=1.3
    curve=Get easyInOutCrv_var()
    alpha=1
   )
   n32 = GetCurveValue(CurveName=l_wrist_up)
   n33 = GetCurveValue(CurveName=l_wrist_dn)
   n34 = MuscleFNC(
    execute=@n7.then_3
    inTarget=@n6
    paramName=Forearm_L
    value=SelectFloat(A=1, B=float + float(A=@n32, B=@n33), bPickA=Get muscleFullON())
    inRangeA=0
    inRangeB=1
    curve=Get easyInOutCrv_var()
    alpha=1
   )
   n35 = GetCurveValue(CurveName=r_wrist_dn)
   n36 = GetCurveValue(CurveName=r_wrist_up)
   > MuscleFNC(
    inTarget=@n34.outTarget
    paramName=Forearm_R
    value=SelectFloat(A=1, B=float + float(A=@n36, B=@n35), bPickA=Get muscleFullON())
    inRangeA=0
    inRangeB=1
    curve=Get easyInOutCrv_var()
    alpha=1
   )
   Set abs(
    execute=@n3.then_2
    abs:double=GetAlphaFromReferenceTransformDelta(SkeletalMeshComponent=GetOwnerSkeletalMeshComponent(), BoneName=spine_04, BaseName=pelvis, InputRange=(X=0,Y=1), TwistAxis="1,0,0").RawOutput
   ) -> (Output_Get:double)
   Set elbow_rot_l(
    execute=@n3.then_0
    elbow_rot_l:double=GetAlphaFromReferenceTransformDelta(SkeletalMeshComponent=GetOwnerSkeletalMeshComponent(), BoneName=lowerarm_l, BaseName=upperarm_l, InputRange=(X=0,Y=1), TwistAxis="1,0,0").RawOutput
   ) -> (Output_Get:double)
   > Set elbow_rot_r(elbow_rot_r:double=GetAlphaFromReferenceTransformDelta(SkeletalMeshComponent=GetOwnerSkeletalMeshComponent(), BoneName=lowerarm_r, BaseName=upperarm_r, InputRange=(X=0,Y=1), TwistAxis="1,0,0").RawOutput) -> (Output_Get:double)
   n37 = GetOwnerSkeletalMeshComponent()
   Set clavicle_fwdBk_l(
    execute=@n3.then_1
    clavicle_fwdBk_l:double=GetAlphaFromReferenceTransformDelta(SkeletalMeshComponent=@n37, BoneName=clavicle_l, BaseName=spine_05, InputRange=(X=0,Y=1), TwistAxis="1,0,0").RawOutput
   ) -> (Output_Get:double)
   > Set clavicle_upDn_l(clavicle_upDn_l:double=GetAlphaFromReferenceTransformDelta(SkeletalMeshComponent=@n37, BoneName=clavicle_l, BaseName=spine_05, DeltaComponentType=EulerY, InputRange=(X=0,Y=1), TwistAxis="1,0,0").RawOutput) -> (Output_Get:double)
   n38 = GetOwnerSkeletalMeshComponent()
   > Set clavicle_fwdBk_r(clavicle_fwdBk_r:double=GetAlphaFromReferenceTransformDelta(SkeletalMeshComponent=@n38, BoneName=clavicle_r, BaseName=spine_05, InputRange=(X=0,Y=1), TwistAxis="1,0,0").RawOutput) -> (Output_Get:double)
   > Set clavicle_upDn_r(clavicle_upDn_r:double=GetAlphaFromReferenceTransformDelta(SkeletalMeshComponent=@n38, BoneName=clavicle_r, BaseName=spine_05, DeltaComponentType=EulerY, InputRange=(X=0,Y=1), TwistAxis="1,0,0").RawOutput) -> (Output_Get:double)
   n39 = Gravity Override State Function DEPRECATED(
    execute=@n1.then_1
    CharacterPartAnimInstance=Self-Reference()
    GameTimeInSeconds=GetGameTimeInSeconds()
    GravityOverrideParams=Get GravityOverrideParams() -> (GravityOverrideParams:Gravity Override Params Struct Structure)
   )
   > Set GravityOverride(GravityOverride=@n39)
   n40 = Sequence!(execute) -> (then_0, then_1, then_2, then_3, then_4)
   Set wrist_bend_l(
    execute=@n40.then_3
    wrist_bend_l:double=GetAlphaFromReferenceTransformDelta(SkeletalMeshComponent=GetOwnerSkeletalMeshComponent(), BoneName=hand_l, BaseName=lowerarm_l, InputRange=(X=0,Y=1), TwistAxis="1,0,0").RawOutput
   ) -> (Output_Get:double)
   > Set wrist_bend_r(wrist_bend_r:double=GetAlphaFromReferenceTransformDelta(SkeletalMeshComponent=GetOwnerSkeletalMeshComponent(), BoneName=hand_r, BaseName=lowerarm_r, InputRange=(X=0,Y=1), TwistAxis="1,0,0").RawOutput) -> (Output_Get:double)
   Set upperarm_up_l(
    execute=@n40.then_2
    upperarm_up_l:double=GetAlphaFromReferenceTransformDelta(
     SkeletalMeshComponent=GetOwnerSkeletalMeshComponent()
     BoneName=upperarm_l
     BaseName=clavicle_l
     DeltaComponentType=EulerY
     InputRange=(X=0,Y=1)
     TwistAxis="1,0,0"
    ).RawOutput
   ) -> (Output_Get:double)
   > Set upperarm_up_r(upperarm_up_r:double=GetAlphaFromReferenceTransformDelta(
     SkeletalMeshComponent=GetOwnerSkeletalMeshComponent()
     BoneName=upperarm_r
     BaseName=clavicle_r
     DeltaComponentType=EulerY
     InputRange=(X=0,Y=1)
     TwistAxis="1,0,0"
    ).RawOutput) -> (Output_Get:double)
   Set abs(
    execute=@n40.then_4
    abs=GetAlphaFromReferenceTransformDelta(SkeletalMeshComponent=GetOwnerSkeletalMeshComponent(), BoneName=spine_04, BaseName=pelvis, InputRange=(X=0,Y=1), TwistAxis="1,0,0").RawOutput
   )
   Set clavicle_fwdBk_l(
    execute=@n40.then_1
    clavicle_fwdBk_l=GetAlphaFromReferenceTransformDelta(SkeletalMeshComponent=GetOwnerSkeletalMeshComponent(), BoneName=clavicle_l, BaseName=spine_05, InputRange=(X=0,Y=1), TwistAxis="1,0,0").RawOutput
   )
   > Set clavicle_fwdBk_r(clavicle_fwdBk_r=GetAlphaFromReferenceTransformDelta(SkeletalMeshComponent=GetOwnerSkeletalMeshComponent(), BoneName=clavicle_r, BaseName=spine_05, InputRange=(X=0,Y=1), TwistAxis="1,0,0").RawOutput)
   Set elbow_rot_l(
    execute=@n40.then_0
    elbow_rot_l=GetAlphaFromReferenceTransformDelta(SkeletalMeshComponent=GetOwnerSkeletalMeshComponent(), BoneName=lowerarm_l, BaseName=upperarm_l, InputRange=(X=0,Y=1), TwistAxis="1,0,0").RawOutput
   )
   > Set elbow_rot_r(elbow_rot_r=GetAlphaFromReferenceTransformDelta(SkeletalMeshComponent=GetOwnerSkeletalMeshComponent(), BoneName=lowerarm_r, BaseName=upperarm_r, InputRange=(X=0,Y=1), TwistAxis="1,0,0").RawOutput)
   Reroute Node()
   > Sequence()
   > n41 = Get Dynamic Material Instance(SkeletalMeshComponent=GetPartSkeletalMeshComponent())
   > n42 = Sequence!(execute) -> (then_0, then_1, then_2, then_3, then_4, then_5:Exec, then_6:Exec)
   n43 = MuscleFNC(
    execute=@n42.then_2
    inTarget=@n41
    paramName=Deltoid_L
    value=Get upperarm_up_l() -> (upperarm_up_l:double)
    inRangeA=0
    inRangeB=60
    curve=Get easyInOutCrv_var()
    alpha=1
   )
   > n44 = MuscleFNC(
    inTarget=@n43.outTarget
    paramName=Deltoid_R
    value=Get upperarm_up_r() -> (upperarm_up_r:double)
    inRangeA=0
    inRangeB=60
    curve=Get easyInOutCrv_var()
    alpha=1
   )
   > n45 = MuscleFNC(inTarget=@n44.outTarget, paramName=Chest L, value=Get clavicle_fwdBk_l(), inRangeA=0, inRangeB=25, curve=Get easyInOutCrv_var(), alpha=1)
   > MuscleFNC(inTarget=@n45.outTarget, paramName=Chest R, value=Get clavicle_fwdBk_r(), inRangeA=0, inRangeB=25, curve=Get easyInOutCrv_var(), alpha=1)
   n46 = MuscleFNC(execute=@n42.then_0, paramName=Biceps_L, value=Get elbow_rot_l(), inRangeA=0, inRangeB=95, curve=Get easyInOutCrv_var(), alpha=1)
   > n47 = MuscleFNC(inTarget=@n46.outTarget, paramName=Biceps_R, value=Get elbow_rot_r(), inRangeA=0, inRangeB=95, curve=Get easyInOutCrv_var(), alpha=1)
   > n48 = MuscleFNC(inTarget=@n47.outTarget, paramName=Triceps_L, value=Get elbow_rot_l(), inRangeA=0, inRangeB=-20, curve=Get easyInOutCrv_var(), alpha=1)
   > MuscleFNC(inTarget=@n48.outTarget, paramName=Triceps_R, value=Get elbow_rot_r(), inRangeA=0, inRangeB=-20, curve=Get easyInOutCrv_var(), alpha=1)
   n49 = MuscleFNC(
    execute=@n42.then_4
    inTarget=@n41
    paramName=Back_L
    value=Get clavicle_fwdBk_l()
    inRangeA=0
    inRangeB=30
    curve=Get easyInOutCrv_var()
    alpha=1
   )
   > n50 = MuscleFNC(inTarget=@n49.outTarget, paramName=Back_R, value=Get clavicle_fwdBk_r(), inRangeA=0, inRangeB=-10, curve=Get easyInOutCrv_var(), alpha=1)
   > n51 = MuscleFNC(
    inTarget=@n50.outTarget
    paramName=Forearm_L
    value=Get wrist_bend_l() -> (wrist_bend_l:double)
    inRangeA=0
    inRangeB=-20
    curve=Get easyInOutCrv_var()
    alpha=1
   )
   > MuscleFNC(
    inTarget=@n51.outTarget
    paramName=Forearm_R
    value=Get wrist_bend_r() -> (wrist_bend_r:double)
    inRangeA=0
    inRangeB=-20
    curve=Get easyInOutCrv_var()
    alpha=1
   )
   MuscleFNC(
    execute=@n42.then_6
    inTarget=@n41
    paramName=Abdomen
    value=Get abs() -> (abs:double)
    inRangeA=0
    inRangeB=35
    curve=Get easyInOutCrv_var()
    alpha=1
   )
   MuscleFNC(inTarget=@n17.outTarget, paramName=Abdomen, value=1, inRangeA=0, inRangeB=20, alpha=1)
   Get abs()
  ]}
 MuscleFNC {nodes [
   n1 = MuscleFNC!() -> (
    then
    inTarget:Material Instance Dynamic Object Reference
    paramName:Name
    value:double=1
    inRangeA:double=~
    inRangeB:double=~
    curve:Curve Float Object Reference
    alpha:double=1
   ) {is=title K2Node_FunctionEntry}
   n2 = MapRangeClamped(Value=@n1.value, InRangeA=@n1.inRangeA, InRangeB=@n1.inRangeB)
   n3 = "Max (Float)"(A=float * float(A=GetFloatValue!(self:Curve Float Object Reference=@n1.curve, InTime:float=@n2) -> (ReturnValue:float), B=@n1.alpha), B=0)
   > SetScalarParameterValue(self=@n1, ParameterName=@n1.paramName, Value=@n3)
   > Return Node!(execute, outValue:double=@n3, outTarget:Material Instance Dynamic Object Reference=@n1)
  ]}
]
variables [
 abs {type=number}
 clavicle_fwdBk_l {type=number}
 clavicle_fwdBk_r {type=number}
 clavicle_upDn_l {type=number}
 clavicle_upDn_r {type=number}
 easyInOutCrv_var {type=CurveFloat, default=/Game/Characters/Player/Male/Large/Bodies/M_LRG_Rumble/Meshes/muscleEasyInOutCrv}
 elbow_rot_l {type=number}
 elbow_rot_r {type=number}
 fort_player {type=FortPlayerAnimInstance, default=null}
 GravityOverride {type=Vector}
 GravityOverrideParams {type=GravityOverrideParamsStruct}
 has_hand_glider {type=boolean}
 IsMuscleON {type=boolean, default=true}
 muscleFullON {type=boolean}
 Skel_Mesh {type=SkeletalMeshComponent}
 upperarm_up_l {type=number}
 upperarm_up_r {type=number}
 wrist_bend_l {type=number}
 wrist_bend_r {type=number}
]
