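# Physics asset for the large male BasilStrong player body, named as a set of colliders for its cape.
# Capsule bodies cover pelvis, upper spine bones, arms, hands, legs and feet, with a second capsule on the pelvis and none on head or neck.
# Constraints chain spine and limbs back to the pelvis, all sharing identical swing and twist limit angles.
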PhysicsAsset /Game/Characters/Player/Male/Large/Bodies/M_LRG_BasilStrong/Meshes/M_LRG_BasilStrong_CapeColliders
bodies {
 calf_l [{shapeType=Capsule, length=26.4868, radius=12.5184, center=<-16.4529 -2.2427 0>, rotation=<rot 90 0 0>}]
 calf_r [{shapeType=Capsule, length=26.4868, radius=12.5184, center=<16.4529 2.2427 0>, rotation=<rot 90 0 180>}]
 foot_l [{shapeType=Capsule, length=15.0766, radius=10.0396, center=<-3.9164 3.986 0>, rotation=<rot 0 14.9998 90>}]
 foot_r [{shapeType=Capsule, length=15.0766, radius=10.0396, center=<3.9164 -3.986 0>, rotation=<rot 0 14.9998 -90>}]
 hand_l [{shapeType=Capsule, length=9.5015, radius=6.7933, center=<6.2703 0 0>, rotation=<rot 90 0 0>}]
 hand_r [{shapeType=Capsule, length=9.5015, radius=6.7933, center=<-6.2703 0 0>, rotation=<rot 90 0 180>}]
 lowerarm_l [{shapeType=Capsule, length=12.2721, radius=9.624, center=<8.9852 0 1.9278>, rotation=<rot 90 0 0>}]
 lowerarm_r [{shapeType=Capsule, length=12.2721, radius=9.624, center=<-8.9852 0 -1.9278>, rotation=<rot 90 0 180>}]
 pelvis [
  {shapeType=Capsule, length=13, radius=15.539, center=<-7.4686 -2.09 0>}
  {shapeType=Capsule, length=15, radius=15, shapeName=pelvis_capsule, center=<-7.4928 0 -6.3344>}
 ]
 spine_02 [{shapeType=Capsule, length=7, radius=16.4999, center=<-5.0595 -1.7096 0>}]
 spine_03 [{shapeType=Capsule, length=10, radius=16.0565}]
 spine_04 [{shapeType=Capsule, length=23.25, radius=18.6297, center=<0 -1.6972 0>}]
 spine_05 [{shapeType=Capsule, length=20, radius=17.5317, center=<0 -1.223 0>}]
 thigh_l [{shapeType=Capsule, length=17.2247, radius=12.8304, center=<-22.1102 0 1.2756>, rotation=<rot 90 0 180>}]
 thigh_r [{shapeType=Capsule, length=17.2247, radius=12.8304, center=<22.1102 0 -1.2756>, rotation=<rot 90 180 180>}]
 upperarm_l [{shapeType=Capsule, length=11.1688, radius=13.7525, center=<10.0455 0 0>, rotation=<rot 90 0 0>}]
 upperarm_r [{shapeType=Capsule, length=11.1688, radius=13.7525, center=<-10.0455 0 0>, rotation=<rot 90 0 180>}]
}
constraints [
 {bone1Name=spine_02, bone2Name=pelvis, swing1LimitDegrees=45, swing2LimitDegrees=45, twistLimitDegrees=45}
 {bone1Name=spine_03, bone2Name=spine_02, swing1LimitDegrees=45, swing2LimitDegrees=45, twistLimitDegrees=45}
 {bone1Name=spine_04, bone2Name=spine_03, swing1LimitDegrees=45, swing2LimitDegrees=45, twistLimitDegrees=45}
 {bone1Name=spine_05, bone2Name=spine_04, swing1LimitDegrees=45, swing2LimitDegrees=45, twistLimitDegrees=45}
 {bone1Name=upperarm_l, bone2Name=spine_05, swing1LimitDegrees=45, swing2LimitDegrees=45, twistLimitDegrees=45}
 {bone1Name=lowerarm_l, bone2Name=upperarm_l, swing1LimitDegrees=45, swing2LimitDegrees=45, twistLimitDegrees=45}
 {bone1Name=hand_l, bone2Name=lowerarm_l, swing1LimitDegrees=45, swing2LimitDegrees=45, twistLimitDegrees=45}
 {bone1Name=calf_l, bone2Name=thigh_l, swing1LimitDegrees=45, swing2LimitDegrees=45, twistLimitDegrees=45}
 {bone1Name=foot_l, bone2Name=calf_l, swing1LimitDegrees=45, swing2LimitDegrees=45, twistLimitDegrees=45}
 {bone1Name=thigh_r, bone2Name=pelvis, swing1LimitDegrees=45, swing2LimitDegrees=45, twistLimitDegrees=45}
 {bone1Name=thigh_l, bone2Name=pelvis, swing1LimitDegrees=45, swing2LimitDegrees=45, twistLimitDegrees=45}
 {bone1Name=upperarm_r, bone2Name=spine_05, swing1LimitDegrees=45, swing2LimitDegrees=45, twistLimitDegrees=45}
 {bone1Name=lowerarm_r, bone2Name=upperarm_r, swing1LimitDegrees=45, swing2LimitDegrees=45, twistLimitDegrees=45}
 {bone1Name=hand_r, bone2Name=lowerarm_r, swing1LimitDegrees=45, swing2LimitDegrees=45, twistLimitDegrees=45}
 {bone1Name=calf_r, bone2Name=thigh_r, swing1LimitDegrees=45, swing2LimitDegrees=45, twistLimitDegrees=45}
 {bone1Name=foot_r, bone2Name=calf_r, swing1LimitDegrees=45, swing2LimitDegrees=45, twistLimitDegrees=45}
]
